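# Sound cue for the Golden Skeleton NPC's vocal reaction to a steal, going by the asset name.
# A random node picks among four equally weighted female scream waves and is excluded from branch culling.
# It plays boosted in volume on the critical enemy VO sound class, using the Golden Skeleton NPC VO attenuation.
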SoundCue /NPCLibrary/NPCs/GoldenSkeleton/GoldenSkeleton/Audio/VO/GOLDENSKELETON_StealReact_Cue
root=Random_1
nodes [
 {
  id=Random_1
  children [WavePlayer_4,WavePlayer_5,WavePlayer_6,WavePlayer_7]
  properties {bShouldExcludeFromBranchCulling=true, weights [1,1,1,1]}
 }
 {id=WavePlayer_4, properties {soundWaveAssetPtr=/NPCLibrary/NPCs/GoldenSkeleton/GoldenSkeleton/Audio/VO/Waves/S16_FEMALEORO_Screams_013}}
 {id=WavePlayer_5, properties {soundWaveAssetPtr=/NPCLibrary/NPCs/GoldenSkeleton/GoldenSkeleton/Audio/VO/Waves/S16_FEMALEORO_Screams_010}}
 {id=WavePlayer_6, properties {soundWaveAssetPtr=/NPCLibrary/NPCs/GoldenSkeleton/GoldenSkeleton/Audio/VO/Waves/S16_FEMALEORO_Screams_011}}
 {id=WavePlayer_7, properties {soundWaveAssetPtr=/NPCLibrary/NPCs/GoldenSkeleton/GoldenSkeleton/Audio/VO/Waves/S16_FEMALEORO_Screams_012}}
]
properties {
 attenuationSettings=&/NPC_VO_GoldenSkeleton
 maxDistance=6000
 soundClassObject=&Fort_VO_Enemy_Critical
 volumeMultiplier=1.7
}
